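# Material function in the frontend Rewards UI materials that converts 2D coordinates into polar (angle, radius) coordinates.
# Input minus one, centered for the default 2x-tiled UV, feeds custom HLSL atan2 nodes for a wrapped 0–1 angle, appended with vector length.
# A reversed output negates X to mirror the angle.
# A third output sharpens that angle's cosine with a high power, producing a narrow mask at the wrap seam.
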MaterialFunction /Game/UI/Frontend/Rewards/materials/VectortoRadialFix
graph {nodes [
  n1 = Subtract(
   A=FunctionInput(Preview=TextureCoordinate() {uTiling=2, vTiling=2}) {bUsePreviewValueAsDefault=true, inputType=FunctionInput_Vector2}
   B=Constant() {r=1}
  )
  MaterialExpressionComponentMask_34 = ComponentMask(None=@n1) {r=true}
  MaterialExpressionComponentMask_35 = ComponentMask(None=@n1) {g=true}
  n2 = Custom!(x=@MaterialExpressionComponentMask_34, y=@MaterialExpressionComponentMask_35) {
   code="return (atan2(y,x));"
   description=atan2
   outputType=CMOT_Float1
   inputs [
    {inputName=x, input {expression=&self:MaterialExpressionComponentMask_34}}
    {inputName=y, input {expression=&self:MaterialExpressionComponentMask_35}}
   ]
  }
  n3 = Divide(A=ComponentMask(None=@n2) {r=true}, B=Constant() {r=6.283})
  n4 = Multiply(A=@n1, B=@n1)
  n5 = SquareRoot(None=Add(A=ComponentMask(None=@n4) {r=true}, B=ComponentMask(None=@n4) {g=true}))
  FunctionOutput(None=AppendVector(A=Frac(None=@n3), B=@n5)) {outputName=Vector to Radial}
  MaterialExpressionMultiply_13 = Multiply(A=ComponentMask(None=@n1) {r=true}) {constB=-1}
  MaterialExpressionComponentMask_37 = ComponentMask(None=@n1) {g=true}
  n6 = Custom(x=@MaterialExpressionMultiply_13, y=@MaterialExpressionComponentMask_37) {
   code="return (atan2(y,x));"
   description=atan2
   outputType=CMOT_Float1
   inputs [
    {inputName=x, input {expression=&self:MaterialExpressionMultiply_13}}
    {inputName=y, input {expression=&self:MaterialExpressionComponentMask_37}}
   ]
  }
  n7 = Frac(None=Divide(A=ComponentMask(None=@n6) {r=true}, B=Constant() {r=6.283}))
  FunctionOutput(None=AppendVector(A=@n7, B=@n5)) {outputName=Vector to Radial Reversed}
  n8 = Cosine(None=@n7)
  n9 = Power(Base=@n8) {constExponent=500}
  FunctionOutput(None=Clamp(None=Multiply(A=@n8, B=@n9))) {outputName=Vector to Reversed Mask}
 ]}
