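# Translucent UI master material for an animated, rarity-colored flare on daily reward screens.
# Four counter-panning radial noise samples, offset by a Random parameter, blend through masks and sum to drive another noise lookup.
# That noise masks an exponential radial gradient, which adds to a smoothstepped rounded-box falloff to form the shape and opacity.
# Color blends two rarity tints by the noise, plus the first tint scaled by that shape.
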Material /Game/UI/Frontend/Rewards/materials/M_UI_DailyReward_Flare-Master
graph {
 properties {
  bUsedWithStaticMesh=false
  blendMode=BLEND_Translucent
  materialDomain=MD_UI
  pixelDepthOffsetMode=PDOM_Legacy
  usedShadingModels=Default Lit
 }
 nodes [
  n1 = VectorParameter() {parameterName=Rarity Color 1, defaultValue=<rgba 0.2958 0.6571 1 1>}
  n2 = Add(
   A=Multiply(A=Time(), B=ScalarParameter() {defaultValue=1, parameterName=MultiplyTime}) {constB=8}
   B=ScalarParameter() {parameterName=Random}
  )
  n3 = Panner(
   Coordinate=ComponentMask(None=MaterialFunctionCall!(In) {materialFunction=&/VectortoRadialFix}.Vector to Radial Reversed) {r=true}
   Time=@n2
  ) {speedX=0.025, speedY=0.05}
  n4 = Panner(
   Coordinate=MaterialFunctionCall() {materialFunction=&VectorToRadialValue}.Vector Converted to Angle
   Time=@n2
  ) {speedX=-0.025, speedY=-0.05}
  n5 = Panner(
   Coordinate=ComponentMask(None=MaterialFunctionCall() {materialFunction=&VectortoRadialFix}.Vector to Radial Reversed) {r=true}
   Time=@n2
  ) {speedX=-0.025, speedY=-0.05}
  n6 = Panner(
   Coordinate=MaterialFunctionCall() {materialFunction=&VectorToRadialValue}.Vector Converted to Angle
   Time=@n2
  ) {speedX=0.025, speedY=0.05}
  n7 = Add(
   A=LinearInterpolate(
    A=TextureSample(UVs=@n3) {samplerType=SAMPLERTYPE_Grayscale, texture=&LowResBlurredNoise}.R
    B=TextureSample(UVs=@n4) {samplerType=SAMPLERTYPE_Grayscale, texture=&LowResBlurredNoise}.R
    Alpha=MaterialFunctionCall() {materialFunction=&VectortoRadialFix}.Vector to Reversed Mask
   )
   B=LinearInterpolate(
    A=TextureSample(UVs=@n5) {samplerType=SAMPLERTYPE_Grayscale, texture=&LowResBlurredNoise}.R
    B=TextureSample(UVs=@n6) {samplerType=SAMPLERTYPE_Grayscale, texture=&LowResBlurredNoise}.R
    Alpha=MaterialFunctionCall() {materialFunction=&VectortoRadialFix}.Vector to Reversed Mask
   )
  )
  n8 = TextureSample(UVs=Multiply(A=ComponentMask(None=@n7) {r=true}, B=ScalarParameter() {defaultValue=1, parameterName=MultiplyUVScale}) {constB=0.5}) {samplerType=SAMPLERTYPE_Grayscale, texture=&LowResBlurredNoise}
  n9 = ScalarParameter() {defaultValue=0.5, parameterName=Shape_Radius}
  n10 = ScalarParameter() {defaultValue=3.4567, parameterName=Shape_Density}
  n11 = Abs(None=Add(A=TextureCoordinate() {uTiling=1.125}, B=Constant2Vector() {g=-0.5, r=-0.5625}) {constB=2})
  n12 = ScalarParameter() {defaultValue=0.05, parameterName=Min}
  n13 = MaterialFunctionCall(Vector 2=Max(A=Subtract(A=@n11, B=@n12)) {constB=0}) {materialFunction=/Engine/Functions/Engine_MaterialFunctions02/Utility/VectorLength}
  n14 = ScalarParameter() {parameterName=Main_Min}
  n15 = ScalarParameter() {defaultValue=0.3532, parameterName=Main_Max}
  n16 = Add(
   A=LinearInterpolate(
    B=MaterialFunctionCall(Radius=@n9, Density=@n10) {materialFunction=&RadialGradientExponential}.RadialGradientExponential
    Alpha=@n8.R
   )
   B=OneMinus(None=MaterialFunctionCall(Alpha=@n13.V2 Length, Min=@n14, Max=@n15) {materialFunction=&SmoothStep}.Result)
  )
  Add(
   A=LinearInterpolate(A=VectorParameter() {parameterName=Rarity Color 3, defaultValue=<rgba 0 0.3745 0.73 1>}.RGB, B=@n1.RGB, Alpha=@n8.R)
   B=Multiply(A=@n1.RGB, B=@n16)
  )
  Multiply(A=@n16, B=ScalarParameter() {defaultValue=1, parameterName=Opacity})
 ]
}
